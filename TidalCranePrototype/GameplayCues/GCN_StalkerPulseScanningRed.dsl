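; Gameplay cue notify that triggers the red stalker pulse scan visual through material parameters.
; On execute it calls the parent, then stamps the current game time as ActivateTime on materials of player 0's Gameplay part skeletal mesh.
; It also writes the cue location, sweep duration, ping radius and activation timestamp to the FringePlank Interrogate Material Parameter Collection, so materials reading it can time the sweep.
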(fn OnExecute
  (params (MyTarget :type Actor%20Object%20Reference) (Parameters :type Gameplay%20Cue%20Parameters%20Structure%20%28by%20ref%29))
  (bind (_normalizedmagnitude _rawmagnitude _effectcontext _matchedtagname _originaltag _aggregatedsourcetags _aggregatedtargettags _location) (Ability|GameplayCue|BreakGameplayCueParameters Parameters))
  (|CallParentFunction MyTarget Parameters)
  (bind _asfort_player_pawn (Utilities|Casting|CastToFortPlayerPawn (Game|GetPlayerPawn 0)))
  (Rendering|Material|SetScalarParameterValueonMaterials (Customization|GetSkeletalMeshforPartType _asfort_player_pawn "Gameplay" :!class /Script/FortniteGame.FortPlayerPawn) "ActivateTime" (Utilities|Time|GetGameTimeinSeconds 0))
  (Rendering|Material|SetVectorParameterValue 0 "/Game/Weapons/Prototype/FringePlank/PostProcess/MPC_FringePlank_Interrogate.MPC_FringePlank_Interrogate" "OriginPositionWS" (Math|Conversions|ToLinearColor(Vector) _location) :!class /Script/Engine.KismetMaterialLibrary)
  (Rendering|Material|SetScalarParameterValue 0 "/Game/Weapons/Prototype/FringePlank/PostProcess/MPC_FringePlank_Interrogate.MPC_FringePlank_Interrogate" "Duration" (Variables|Default|GetSweepDuration) :!class /Script/Engine.KismetMaterialLibrary)
  (Rendering|Material|SetScalarParameterValue 0 "/Game/Weapons/Prototype/FringePlank/PostProcess/MPC_FringePlank_Interrogate.MPC_FringePlank_Interrogate" "Radius" (ScalableFloat|GetValueatLevel (Variables|Default|GetPingRadius)) :!class /Script/Engine.KismetMaterialLibrary)
  (Rendering|Material|SetScalarParameterValue 0 "/Game/Weapons/Prototype/FringePlank/PostProcess/MPC_FringePlank_Interrogate.MPC_FringePlank_Interrogate" "ActivatedTimestamp" (Utilities|Time|GetGameTimeinSeconds 0) :!class /Script/Engine.KismetMaterialLibrary)
  (return false))

(fn ConstructionScript)
